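# Looping outdoor daytime ambience cue for the Chrome biome (Level 1, variation 02), using Athena ambient attenuation and sound class.
# A Quality Level node plays the quad wave for its first two levels and the stereo wave for the last two.
# Each wave runs through two chained, endlessly looping Envelopers: a slow pitch wobble, then a volume drift.
# The quad's volume curve is longer and more irregular than the stereo's.
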SoundCue /Game/Sounds/Artemis/Biomes/Chrome/Biome_Chrome_Day_Loop_Lvl1_02_Ext_Cue
root=QualityLevel_0
nodes [
 {
  id=Enveloper_0
  children [WavePlayer_3]
  properties {
   bLoop=true
   bLoopIndefinitely=true
   loopEnd=60
   pitchCurve {editorCurveData {keys [
      {interpMode=RCIM_Cubic, time=0}
      <at 15 1.02 RCIM_Cubic>
      {arriveTangent=0, interpMode=RCIM_Cubic, leaveTangent=0, tangentMode=RCTM_User, time=30}
      <at 45 1.05 RCIM_Cubic>
      {time=60}
     ]}}
   volumeCurve {editorCurveData {keys [<at 0 1>]}}
  }
 }
 {
  id=Enveloper_1
  children [WavePlayer_2]
  properties {
   bLoop=true
   bLoopIndefinitely=true
   loopEnd=60
   pitchCurve {editorCurveData {keys [
      {interpMode=RCIM_Cubic, time=0}
      <at 15 1.02 RCIM_Cubic>
      {arriveTangent=0, interpMode=RCIM_Cubic, leaveTangent=0, tangentMode=RCTM_User, time=30}
      <at 45 1.05 RCIM_Cubic>
      {time=60}
     ]}}
   volumeCurve {editorCurveData {keys [<at 0 1>]}}
  }
 }
 {
  id=Enveloper_2
  children [Enveloper_1]
  properties {
   bLoop=true
   bLoopIndefinitely=true
   loopEnd=60
   pitchCurve {editorCurveData {keys [{time=0}]}}
   volumeCurve {editorCurveData {keys [
      {arriveTangent=-0.01, interpMode=RCIM_Cubic, time=0, value=1}
      <at 10 0.8 RCIM_Cubic>
      {arriveTangent=-0.0007, interpMode=RCIM_Cubic, leaveTangent=-0.0007, tangentMode=RCTM_User, time=20, value=1}
      {arriveTangent=0.0003, interpMode=RCIM_Cubic, leaveTangent=0.0003, tangentMode=RCTM_User, time=25, value=0.651}
      {arriveTangent=0.0014, interpMode=RCIM_Cubic, leaveTangent=0.0014, tangentMode=RCTM_User, time=34.1558, value=0.9045}
      {arriveTangent=0.0007, interpMode=RCIM_Cubic, leaveTangent=0.0007, tangentMode=RCTM_User, time=40.5998, value=0.8549}
      {arriveTangent=-0.0009, interpMode=RCIM_Cubic, leaveTangent=-0.0009, tangentMode=RCTM_User, time=46.5069, value=0.95}
      {arriveTangent=0.0013, interpMode=RCIM_Cubic, leaveTangent=0.0013, tangentMode=RCTM_User, time=50.5881, value=0.7868}
      <at 60 1>
     ]}}
  }
 }
 {
  id=Enveloper_3
  children [Enveloper_0]
  properties {
   bLoop=true
   bLoopIndefinitely=true
   loopEnd=30
   pitchCurve {editorCurveData {keys [{time=0}]}}
   volumeCurve {editorCurveData {keys [
      {arriveTangent=-0.01, interpMode=RCIM_Cubic, time=0, value=1}
      <at 10 0.8 RCIM_Cubic>
      {interpMode=RCIM_Cubic, leaveTangent=0.01, time=20, value=1}
     ]}}
  }
 }
 {id=QualityLevel_0, children [Enveloper_2,Enveloper_2,Enveloper_3,Enveloper_3]}
 {id=WavePlayer_2, properties {bLooping=true, soundWaveAssetPtr=&/AMB_Chrome_Day_Loop_Lvl1_02_Quad}}
 {id=WavePlayer_3, properties {bLooping=true, soundWaveAssetPtr=&/AMB_Chrome_Day_Loop_Lvl1_02_Stereo}}
]
properties {
 attenuationSettings=&Fort_Ambient_Athena_Att
 maxDistance=2097152
 pitchMultiplier=0.9
 soundClassObject=&Fort_Ambient_Outdoors_Athena
 volumeMultiplier=0.551
 preEffectBusSends [{}]
}
